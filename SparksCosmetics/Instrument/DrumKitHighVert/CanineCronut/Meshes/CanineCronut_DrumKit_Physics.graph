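# Physics asset for the CanineCronut drum kit instrument cosmetic, giving its dangling loop and jack bones simulated secondary motion.
# Capsule bodies form a loop chain from the kickdrum that branches at the center link into top and bottom segments, plus left and right two-bone jack chains.
# These chains lock twist and limit swing per joint, while the kickdrum, with one large capsule, has free swing and twist relative to the handle.
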PhysicsAsset /SparksCosmetics/Instrument/DrumKitHighVert/CanineCronut/Meshes/CanineCronut_DrumKit_Physics
bodies {
 dyn_bottom_loop_01 [{
   shapeType=Capsule
   length=19.5985
   radius=3
   shapeName=dyn_bottom_loop_01_capsule
   center=<2.9995 -0.5454 1.42>
   rotation=<rot 0.0238 0.0064 -111.0103>
  }]
 dyn_bottom_loop_02 [{
   shapeType=Capsule
   length=10.1374
   radius=3
   shapeName=dyn_bottom_loop_02_capsule
   center=<3.3855 0.0724 0.3663>
   rotation=<rot 0.0001 0 85.3599>
  }]
 dyn_jack_l_01 [{
   shapeType=Capsule
   length=2.8318
   radius=1.0457
   shapeName=dyn_jack_l_01_capsule
   center=<1.5289 0.077 -0.1643>
   rotation=<rot 0 -93.3922 -80.8335>
  }]
 dyn_jack_l_02 [{shapeType=Capsule, length=6.9461, radius=0.9321, shapeName=dyn_jack_l_02_capsule, center=<4 0.25 0>, rotation=<rot 0 -87.041 -84.534>}]
 dyn_jack_r_01 [{shapeType=Capsule, length=2.6109, radius=0.9058, shapeName=dyn_jack_r_01_capsule, center=<1.1745 0.077 0.0815>, rotation=<rot 0 90 90>}]
 dyn_jack_r_02 [{shapeType=Capsule, length=7.6413, radius=0.9187, shapeName=dyn_jack_r_02_capsule, center=<3 0 0>, rotation=<rot 0 -91.1241 -78.8782>}]
 dyn_loop_01 [{shapeType=Capsule, length=7.1703, radius=3.9704, shapeName=dyn_loop_01_capsule, center=<1 -1 0.3>, rotation=<rot 0 0 -90>}]
 dyn_loop_02 [{shapeType=Capsule, length=13.5166, radius=3.3994, shapeName=dyn_loop_02_capsule, rotation=<rot 0 2.9191 -90>}]
 dyn_loop_03 [{shapeType=Capsule, length=17.1985, radius=4.1908, shapeName=dyn_loop_03_capsule, center=<0 0 3.8439>, rotation=<rot 0.0001 6.6899 -90.0001>}]
 dyn_loop_center_04 [{
   shapeType=Capsule
   length=24.3548
   radius=4.1023
   shapeName=dyn_loop_center_04_capsule
   center=<0.7528 -2.969 3.2993>
   rotation=<rot -0.9009 -5.757 -106.0387>
  }]
 dyn_top_loop_01 [{shapeType=Capsule, length=20.3398, radius=3, shapeName=dyn_top_loop_01_capsule, center=<3 -2.5794 0.4318>, rotation=<rot 0 0 -112.032>}]
 dyn_top_loop_02 [{shapeType=Capsule, length=14.9119, radius=3, shapeName=dyn_top_loop_02_capsule, center=<2.5 0 1.5>, rotation=<rot 0.0763 1.5254 87.1379>}]
 handle [{shapeType=Capsule, length=0.505, radius=0.505, rotation=<rot 0 -90 0>}]
 kickdrum [{
   shapeType=Capsule
   length=20.6213
   radius=19.1442
   shapeName=kickdrum_capsule
   center=<-38.9987 6.0692 39.2187>
   rotation=<rot 85.9643 -0.0508 -0.0506>
  }]
}
constraints [
 {bone1Name=dyn_loop_01, bone2Name=kickdrum, swing1LimitDegrees=10, swing2LimitDegrees=10, twistLimitDegrees=45, twistMotion=Locked}
 {bone1Name=dyn_loop_02, bone2Name=dyn_loop_01, swing1LimitDegrees=15, swing2LimitDegrees=30, twistLimitDegrees=45, twistMotion=Locked}
 {bone1Name=dyn_loop_03, bone2Name=dyn_loop_02, swing1LimitDegrees=25, swing2LimitDegrees=35, twistLimitDegrees=45, twistMotion=Locked}
 {bone1Name=dyn_loop_center_04, bone2Name=dyn_loop_03, swing1LimitDegrees=35, swing2LimitDegrees=35, twistLimitDegrees=45, twistMotion=Locked}
 {
  bone1Name=dyn_bottom_loop_01
  bone2Name=dyn_loop_center_04
  swing1LimitDegrees=35
  swing2LimitDegrees=10
  twistLimitDegrees=45
  twistMotion=Locked
 }
 {
  bone1Name=dyn_bottom_loop_02
  bone2Name=dyn_bottom_loop_01
  swing1LimitDegrees=35
  swing2LimitDegrees=45
  twistLimitDegrees=45
  twistMotion=Locked
 }
 {
  bone1Name=dyn_top_loop_01
  bone2Name=dyn_loop_center_04
  swing1LimitDegrees=35
  swing2LimitDegrees=10
  twistLimitDegrees=45
  twistMotion=Locked
 }
 {bone1Name=dyn_top_loop_02, bone2Name=dyn_top_loop_01, swing1LimitDegrees=35, swing2LimitDegrees=30, twistLimitDegrees=45, twistMotion=Locked}
 {bone1Name=dyn_jack_r_02, bone2Name=dyn_jack_r_01, swing1LimitDegrees=45, swing2LimitDegrees=45, twistLimitDegrees=45, twistMotion=Locked}
 {bone1Name=dyn_jack_l_01, bone2Name=kickdrum, swing1LimitDegrees=45, swing2LimitDegrees=45, twistLimitDegrees=45, twistMotion=Locked}
 {bone1Name=dyn_jack_l_02, bone2Name=dyn_jack_l_01, swing1LimitDegrees=45, swing2LimitDegrees=45, twistLimitDegrees=45, twistMotion=Locked}
 {bone1Name=dyn_jack_r_01, bone2Name=kickdrum, swing1LimitDegrees=45, swing2LimitDegrees=45, twistLimitDegrees=45, twistMotion=Locked}
 {
  bone1Name=kickdrum
  bone2Name=handle
  swing1LimitDegrees=45
  swing1Motion=Free
  swing2LimitDegrees=45
  swing2Motion=Free
  twistLimitDegrees=45
  twistMotion=Free
 }
]
