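# Animation Blueprint for the CanineCronut drum kit cosmetic that adds physics-driven secondary motion to its skeleton.
# The AnimGraph runs the input pose through a single RigidBody node using the kit's override physics asset, then outputs the result.
# Simulation runs in root-bone space with damped, capped motion transfer, overridden gravity and a bool-driven alpha.
# The event graph holds only unconnected stub nodes, and there are no variables.
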AnimBlueprint /SparksCosmetics/Instrument/DrumKitHighVert/CanineCronut/Meshes/CanineCronut_DrumKit_AnimBP
target_skeleton=&/CanineCronut_DrumKit_Skeleton
graphs [
 AnimGraph {nodes [
   n1 = RigidBody!(ComponentPose=Local To Component(LocalPose=Input Pose()), bAlphaBoolEnabled:bool, ExternalForce) -> (Pose) {
    alphaInputType=Bool
    bClampLinearTranslationLimitToRefPose=true
    bOverrideWorldGravity=true
    overridePhysicsAsset=&/CanineCronut_DrumKit_Physics
    simulationSpace=BaseBoneSpace
    alphaBoolBlend {blendOption=Linear}
    alphaScaleBiasClamp {bInterpResult=true, interpSpeedDecreasing=6, interpSpeedIncreasing=7}
    baseBoneRef {boneName=root}
    componentAppliedLinearAccClamp=<300 300 300>
    componentLinearAccScale=<0.15 0.15 0.15>
    componentLinearVelScale=<0.15 0.15 0.15>
    overrideWorldGravity {z=-980}
    simSpaceSettings {
     dampingAlpha=0.8
     maxAngularAcceleration=10
     maxAngularVelocity=6
     maxLinearAcceleration=100
     maxLinearVelocity=100
     velocityScaleZ=0
     worldAlpha=0.7
     externalLinearDragV=<0.1 0.1 0.1>
    }
   }
   Output Pose(Result=Component To Local(ComponentPose=@n1))
  ]}
 EventGraph {nodes [
   TryGetPawnOwner()
   Event BlueprintUpdateAnimation()
  ]}
]
variables []
